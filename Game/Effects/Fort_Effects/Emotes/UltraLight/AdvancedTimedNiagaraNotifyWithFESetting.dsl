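; Timed Niagara anim notify state, filed with UltraLight emote effects, that tells its spawned effect whether it is displayed in a frontend context.
; On notify begin it calls the parent, then casts the effect spawned for the mesh to a Niagara component.
; It writes the mesh's cosmetic frontend check to both the bIsFrontEnd and bIsFrontEndPreview user variables and returns the parent's result.
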(fn Received_NotifyBegin
  (params (MeshComp :type Skeletal%20Mesh%20Component%20Object%20Reference) (Animation :type Anim%20Sequence%20Base%20Object%20Reference) (TotalDuration :type Float%20%28single-precision%29) (EventReference :type Anim%20Notify%20Event%20Reference%20Structure%20%28by%20ref%29))
  (bind _meshcomp (Variables|GetMeshComp))
  (bind _returnvalue (Reactive|Frontend|IsCosmeticObjectDisplayedinFrontendContext _meshcomp))
  (bind _returnvalue_2 (|CallParentFunction MeshComp Animation TotalDuration EventReference))
  (bind _returnvalue_3 (AnimNotify|GetSpawnedEffect MeshComp))
  (bind _asniagara_particle_system_component (Utilities|Casting|CastToNiagaraComponent _returnvalue_3))
  (Niagara|SetNiagaraVariable(Bool) _asniagara_particle_system_component "bIsFrontEnd" _returnvalue)
  (Niagara|SetNiagaraVariable(Bool) _asniagara_particle_system_component "bIsFrontEndPreview" _returnvalue)
  (return _returnvalue_2))
