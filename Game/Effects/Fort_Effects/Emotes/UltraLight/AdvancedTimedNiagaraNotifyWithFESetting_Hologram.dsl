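; Timed Niagara anim notify state that makes its spawned effect aware of the frontend and hides it on frontend hologram pawns.
; On notify begin it calls the parent, then retrieves the spawned Niagara component.
; It feeds the mesh's frontend-display check into the bIsFrontEnd and bIsFrontEndPreview Niagara booleans.
; If the mesh owner is a FortPlayerPawn, it sets effect visibility to the inverse of bIsInFrontEndHologram.
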(fn Received_NotifyBegin
  (params (MeshComp :type Skeletal%20Mesh%20Component%20Object%20Reference) (Animation :type Anim%20Sequence%20Base%20Object%20Reference) (TotalDuration :type Float%20%28single-precision%29) (EventReference :type Anim%20Notify%20Event%20Reference%20Structure%20%28by%20ref%29))
  (bind _meshcomp (Variables|GetMeshComp))
  (bind _returnvalue (Reactive|Frontend|IsCosmeticObjectDisplayedinFrontendContext _meshcomp))
  (bind _returnvalue_2 (|CallParentFunction MeshComp Animation TotalDuration EventReference))
  (bind _returnvalue_3 (AnimNotify|GetSpawnedEffect MeshComp))
  (Utilities|IsValid _returnvalue_3
    (:Is%20Valid
      (bind _asniagara_particle_system_component (Utilities|Casting|CastToNiagaraComponent _returnvalue_3)
        (:then
          (Niagara|SetNiagaraVariable(Bool) _asniagara_particle_system_component "bIsFrontEnd" _returnvalue)
          (Niagara|SetNiagaraVariable(Bool) _asniagara_particle_system_component "bIsFrontEndPreview" _returnvalue)
          (bind _asfort_player_pawn (Utilities|Casting|CastToFortPlayerPawn (Components|GetOwner (Variables|GetMeshComp)))
            (:then
              (if (Variables|/Script/FortniteGame.FortPlayerPawn|GetbIsInFrontEndHologram _asfort_player_pawn)
                (Rendering|SetVisibility _returnvalue_3 false false)
                (else
                  (Rendering|SetVisibility _returnvalue_3 true false))))
            (:CastFailed
              _)))
        (:CastFailed
          _)))
    (:Is%20Not%20Valid
      _))
  (return _returnvalue_2))
