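# Translucent, unlit, two-sided particle material for the energy-absorb effect in the UltraLight emote, drawing animated plasma on Niagara meshes.
# Two noise layers, plus an optional third, distort slope-skewed UVs for the plasma and optional second texture, timed by a particle Timer when positive.
# Color comes from a curve atlas or solid color, then gradient-based hue shift.
# Refined plasma, a noise-perturbed V edge mask, particle alpha and depth fade shape opacity.
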Material /Game/Effects/Fort_Effects/Emotes/UltraLight/Material/M_AbsorbEnergy
graph {
 properties {
  bUsedWithNiagaraMeshParticles=true
  blendMode=BLEND_Translucent
  pixelDepthOffsetMode=PDOM_Legacy
  shadingModel=MSM_Unlit
  twoSided=true
  usedShadingModels=Unlit
 }
 nodes [
  n1 = ParticleColor()
  n2 = TextureObjectParameter() {
   group=---TextureA
   parameterName=TextureA_Texture
   samplerType=SAMPLERTYPE_LinearColor
   texture=/Game/Effects/Fort_Effects/Emotes/UltraLight/Texture/T_Plasma_A_C
  }
  n3 = VectorParameter() {
   group=---TextureA
   parameterName=TextureA_UVControl_Mark:********************
   channelNames {a=PanV, b=PanU, g=ScaleV, r=ScaleU}
   defaultValue=<rgba 1 1 0 0>
  }
  n4 = DynamicParameter() {defaultValue=<rgba 0 1 1 1>, paramNames [Timer,Param2,Param3,Param4]}
  n5 = Time() {bOverride_Period=true, period=13333}
  n6 = If(A=@n4.Timer, B=Constant(), "A > B"=@n4.Timer, "A == B"=@n5, "A < B"=@n5)
  n7 = TextureObjectParameter() {group=---NoiseA, parameterName=NoiseA_Texture, texture=/Game/Characters/Player/Male/Medium/Bodies/M_MED_Viper/Textures/Glitch/T_SoftNoise}
  n8 = VectorParameter() {
   group=---NoiseA
   parameterName=NoiseA_UVControl_Mark:********************
   channelNames {a=PanV, b=PanU, g=ScaleV, r=ScaleU}
   defaultValue=<rgba 1 1 0 0>
  }
  n9 = MaterialFunctionCall(Float2=TextureCoordinate()) {materialFunction=&BreakOutFloat2Components}
  n10 = ScalarParameter() {parameterName=Slope}
  n11 = Add(A=TextureCoordinate(), B=AppendVector(A=Multiply(A=@n9.G, B=@n10), B=Constant()))
  n12 = VectorParameter() {
   group=---NoiseA
   parameterName=NoiseA_ValueControl_Mark:--------------------
   sortPriority=33
   channelNames {a=StrengthV, b=StrengthU, g=StrengthOffsetV, r=StrengthOffsetU}
   defaultValue=<rgba -0.5 -0.5 0 0>
  }
  n13 = MaterialFunctionCall!(
   NoiseValue=MaterialFunctionCall(Texture=@n7, UVControl_RGB=@n8, UVControl_A=@n8, "[Override]TimeInput"=@n6, "[Override]UVInput"=@n11) {materialFunction=&FXCommon_TextureSample}.RG
   Control_RGB=@n12
   Control_A=@n12
  ) {materialFunction=/Game/Effects/Niagara/MasterMaterials/Functions/Common/FXCommon_NoiseControl}
  n14 = TextureObjectParameter() {
   group=---NoiseB
   parameterName=NoiseB_Texture
   samplerType=SAMPLERTYPE_LinearGrayscale
   texture=/Game/Effects/Fort_Effects/Emotes/UltraLight/Texture/T_DirectionalNoise_Greyscale
  }
  n15 = VectorParameter() {
   group=---NoiseB
   parameterName=NoiseB_UVControl_Mark:********************
   channelNames {a=PanV, b=PanU, g=ScaleV, r=ScaleU}
   defaultValue=<rgba 1 1 0 0>
  }
  n16 = Power(
   Base=MaterialFunctionCall(Texture=@n14, UVControl_RGB=@n15, UVControl_A=@n15, "[Override]TimeInput"=@n6, "[Override]UVInput"=@n11) {materialFunction=&FXCommon_TextureSample}.R
   Exp=ScalarParameter() {defaultValue=1, group=---NoiseB, parameterName=NoiseB_Power}
  )
  n17 = VectorParameter() {
   group=---NoiseB
   parameterName=NoiseB_ValueControl_Mark:--------------------
   sortPriority=33
   channelNames {a=StrengthV, b=StrengthU, g=StrengthOffsetV, r=StrengthOffsetU}
   defaultValue=<rgba -0.5 -0.5 0 0>
  }
  n18 = MaterialFunctionCall(NoiseValue=AppendVector(A=@n16, B=@n16), Control_RGB=@n17, Control_A=@n17) {materialFunction=&FXCommon_NoiseControl}
  n19 = VectorParameter() {
   group=---NoiseC
   parameterName=NoiseC_UVControl_Mark:********************
   channelNames {a=PanV, b=PanU, g=ScaleV, r=ScaleU}
   defaultValue=<rgba 1 1 0 0>
  }
  n20 = Power(
   Base=MaterialFunctionCall(
    Texture=TextureObjectParameter() {group=---NoiseC, parameterName=NoiseC_Texture, sortPriority=31}
    UVControl_RGB=@n19
    UVControl_A=@n19
    "[Override]TimeInput"=@n6
    "[Override]UVInput"=@n11
   ) {materialFunction=&FXCommon_TextureSample}.R
   Exp=ScalarParameter() {defaultValue=1, group=---NoiseC, parameterName=NoiseC_Power}
  )
  n21 = VectorParameter() {
   group=---NoiseC
   parameterName=NoiseC_ValueControl_Mark:--------------------
   sortPriority=33
   channelNames {a=StrengthV, b=StrengthU, g=StrengthOffsetV, r=StrengthOffsetU}
   defaultValue=<rgba -0.5 -0.5 0 0>
  }
  n22 = Add(
   A=Add(A=Add(A=@n13.NoiseValue, B=@n11), B=@n18.NoiseValue)
   B=StaticSwitchParameter(
    True=MaterialFunctionCall(NoiseValue=@n20, Control_RGB=@n21, Control_A=@n21) {materialFunction=&FXCommon_NoiseControl}.NoiseValue
    False=Constant()
   ) {parameterName=UseNoiseC}
  )
  n23 = MaterialFunctionCall(Texture=@n2, UVControl_RGB=@n3, UVControl_A=@n3, "[Override]TimeInput"=@n6, "[Override]UVInput"=@n22) {materialFunction=&FXCommon_TextureSample}
  n24 = DynamicParameter() {parameterIndex=3, defaultValue=<rgba 0 1 1 1>, paramNames [RefineOpacityPowerOffset, Param2, Param3, Param4]}
  n25 = Constant()
  n26 = VectorParameter() {parameterName=TextureA_Refine_Opacity, defaultValue=<rgba 3 1 0.7 0.5>}
  n27 = MaterialFunctionCall(
   Texture=@n23.R
   RGB=Add(
    A=MaterialFunctionCall(X=@n24.RefineOpacityPowerOffset, Y=@n25, Z=@n25) {materialFunction=&MakeFloat3}.Result
    B=@n26.RGB
   )
   A=@n26.A
  ) {materialFunction=&MF_FX_Refine}
  n28 = MaterialFunctionCall(Float2=TextureCoordinate()) {materialFunction=&BreakOutFloat2Components}
  n29 = Subtract(A=Add(A=@n13.NoiseValue, B=@n18.NoiseValue), B=ScalarParameter() {parameterName=NoiseAFalloff})
  n30 = ScalarParameter() {parameterName=NoiseAToEdgeMask}
  n31 = MaterialFunctionCall!(
   UorV=Add(A=@n28.G, B=Multiply(A=@n29, B=@n30))
   Invert_EdgeControl=VectorParameter() {parameterName=EdgeMask2, defaultValue=<rgba 100 1 0 1>}.RGB
   NoInvert_EdgeControl=VectorParameter() {parameterName=EdgeMask, defaultValue=<rgba 100 1 0 1>}.RGB
  ) {materialFunction=/Game/Effects/Niagara/MasterMaterials/Functions/Common/FXCommon_EdgeMask_Ex}
  DepthFade(Opacity=Multiply(A=@n1.A, B=Multiply(A=@n27.Result, B=@n31.Mask)), FadeDistance=ScalarParameter() {defaultValue=1.5, parameterName=FD})
  n32 = VectorParameter() {parameterName=EdgeHueShift, defaultValue=<rgba 0 0 0 1>}
  n33 = Add(A=LinearInterpolate(A=@n28.G, B=OneMinus(None=@n28.G), Alpha=@n32.A), B=@n32.R)
  n34 = VectorParameter() {
   group=---TextureB
   parameterName=TextureB_UVControl_Mark:********************
   channelNames {a=PanV, b=PanU, g=ScaleV, r=ScaleU}
   defaultValue=<rgba 1 1 0 0>
  }
  n35 = MaterialFunctionCall(
   Texture=TextureObjectParameter() {group=---TextureB, parameterName=TextureB_Texture, sortPriority=31}
   UVControl_RGB=@n34
   UVControl_A=@n34
   "[Override]TimeInput"=@n6
   "[Override]UVInput"=@n22
  ) {materialFunction=&FXCommon_TextureSample}
  n36 = VectorParameter() {parameterName=TextureB_Refine_Brightness, defaultValue=<rgba 10 2000 0.3 0.5>}
  n37 = MaterialFunctionCall(Texture=@n35.R, RGB=@n36.RGB, A=@n36.RGB) {materialFunction=&MF_FX_Refine}
  n38 = ScalarParameter() {defaultValue=1, parameterName=TextureB_MultiplyAdd}
  n39 = Multiply(A=@n23.R, B=ScalarParameter() {defaultValue=1, parameterName=TextureA_ColorRemapMultiply})
  n40 = Multiply(A=Multiply(A=@n35.R, B=ScalarParameter() {defaultValue=1, parameterName=TextureB_ColorRemapMultiply}), B=@n39)
  n41 = CurveAtlasRowParameter(CurveTime=Saturate(None=StaticSwitchParameter(True=@n40, False=@n39) {parameterName=UseTextureB})) {
   atlas=/Game/Effects/Fort_Effects/Emotes/UltraLight/Texture/Curve/CLA_UltraLight
   curve=/Game/Effects/Fort_Effects/Emotes/UltraLight/Texture/Curve/CL_UltraLight_Energy
  }
  n42 = VectorParameter() {parameterName=TextureA_Refine_Brightness, defaultValue=<rgba 10 2000 0.3 0.5>}
  n43 = MaterialFunctionCall(Texture=@n23.R, RGB=@n42.RGB, A=@n42.RGB) {materialFunction=&MF_FX_Refine}
  n44 = Multiply(
   A=StaticSwitchParameter(True=VectorParameter() {parameterName=SolidColor, defaultValue=<rgba 0 0 0 1>}.RGB, False=@n41) {parameterName=UseSolidColor}
   B=Multiply(B=Add(A=@n43.Result)) {constA=0.5}
  )
  n45 = StaticSwitchParameter(True=Multiply(A=Add(A=@n37.Result, B=@n38) {constB=0.1}, B=@n44), False=@n44) {parameterName=UseTextureB}
  MaterialFunctionCall(Emissive=MaterialFunctionCall(Hue Shift Percentage=Multiply(A=@n33, B=@n32.G), Texture=Multiply(A=@n1.RGB, B=@n45)) {materialFunction=&HueShift}.Result) {materialFunction=&MF_FX_EmissiveAdjustments}
 ]
}
