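# Translucent unlit Niagara sprite material for the UltraLight emote, drawing plasma and glow where scene geometry sits near the sprite's depth.
# Plasma texture, warped by noise and scene normals, is masked by depth proximity and tinted by particle color.
# Custom HLSL takes jittered radial scene-depth samples for a soft glow, added to plasma and rim opacity.
# Radial gradient and depth fade shape opacity; forward and mobile paths output zero.
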Material /Game/Effects/Fort_Effects/Emotes/UltraLight/Material/M_DepthGlow
graph {
 properties {bUsedWithNiagaraSprites=true, blendMode=BLEND_Translucent, pixelDepthOffsetMode=PDOM_Legacy, shadingModel=MSM_Unlit, usedShadingModels=Unlit}
 nodes [
  n1 = ParticleColor()
  n2 = DynamicParameter() {paramNames [T_Opacity,T_Brightness,Glow_Opacity,Param4]}
  n3 = TextureObjectParameter() {
   group=---TextureA
   parameterName=TextureA_Texture
   samplerType=SAMPLERTYPE_LinearColor
   texture=/Game/Effects/Fort_Effects/Emotes/UltraLight/Texture/T_Plasma_D_C
  }
  n4 = VectorParameter() {group=---TextureA, parameterName=TextureA_UVOffset, channelNames {a="", b="", g=OffsetV, r=OffsetU}, defaultValue=<rgba 1 1 0 0>}
  n5 = VectorParameter() {
   group=---TextureA
   parameterName=TextureA_UVControl_Mark:********************
   channelNames {a=PanV, b=PanU, g=ScaleV, r=ScaleU}
   defaultValue=<rgba 0.75 0.75 0.3 0.3>
  }
  n6 = SceneTexture() {sceneTextureId=PPI_WorldNormal}
  n7 = ComponentMask(None=Transform(None=@n6.Color) {transformSourceType=TRANSFORMSOURCE_World, transformType=TRANSFORM_Camera}) {g=true, r=true}
  n8 = ScalarParameter() {defaultValue=0.3}
  n9 = TextureCoordinate()
  n10 = TextureObjectParameter() {
   group=---NoiseA
   parameterName=NoiseA_Texture
   samplerType=SAMPLERTYPE_LinearGrayscale
   sortPriority=31
   texture=&T_SimplePerlin_Greyscale
  }
  n11 = VectorParameter() {
   group=---NoiseA
   parameterName=NoiseA_UVControl_Mark:********************
   channelNames {a=PanV, b=PanU, g=ScaleV, r=ScaleU}
   defaultValue=<rgba 1 1 0.07 0.07>
  }
  n12 = MaterialFunctionCall(Texture=@n10, UVControl_RGB=@n11, UVControl_A=@n11) {materialFunction=&FXCommon_TextureSample}
  n13 = VectorParameter() {
   group=---NoiseA
   parameterName=NoiseA_ValueControl_Mark:--------------------
   sortPriority=33
   channelNames {a=StrengthV, b=StrengthU, g=StrengthOffsetV, r=StrengthOffsetU}
   defaultValue=<rgba -0.5 -0.5 0.5 0.5>
  }
  n14 = MaterialFunctionCall!(NoiseValue=AppendVector(A=@n12.R, B=@n12.R), Control_RGB=@n13, Control_A=@n13) {materialFunction=/Game/Effects/Niagara/MasterMaterials/Functions/Common/FXCommon_NoiseControl}
  n15 = MaterialFunctionCall(
   Texture=@n3
   AdditionalUVOffset=AppendVector(A=@n4.OffsetU, B=@n4.OffsetU)
   UVControl_RGB=@n5
   UVControl_A=@n5
   "[Override]UVInput"=Add(A=Multiply(A=@n7, B=@n8), B=Add(A=@n9, B=@n14.NoiseValue))
  ) {materialFunction=&FXCommon_TextureSample}
  n16 = Power(Base=@n15.R, Exp=ScalarParameter() {defaultValue=3, parameterName=TextureA_PowerToColor})
  n17 = ScalarParameter() {defaultValue=5, parameterName=TextureA_StrengthToColor}
  n18 = SceneDepth()
  MaterialExpressionPixelDepth_0 = PixelDepth()
  n19 = Subtract(A=@n18.R, B=@MaterialExpressionPixelDepth_0.R)
  n20 = ScalarParameter() {defaultValue=25, parameterName=TextureD}
  n21 = Saturate(None=OneMinus(None=Divide(A=@n19, B=@n20) {constB=20}))
  n22 = OneMinus(None=Power(Base=OneMinus(None=@n21), Exp=ScalarParameter() {defaultValue=5, parameterName=Texture_DepthMask}))
  n23 = Multiply(A=Multiply(A=@n2.T_Brightness, B=Multiply(A=@n16, B=@n17)), B=@n22)
  n24 = Multiply(A=@n1.RGB, B=Multiply(A=VectorParameter() {parameterName=Color, defaultValue=<rgba 1 1 1 1>}.RGB, B=Add(A=@n23)))
  n25 = Constant()
  MaterialFunctionCall(Emissive=ShadingPathSwitch(Default=@n24, Deferred=@n24, Forward=@n25, Mobile=@n25)) {materialFunction=&MF_FX_EmissiveAdjustments}
  n26 = VectorParameter() {parameterName=RadialGradient, defaultValue=<rgba 0.5 1 0 1>}
  n27 = MaterialFunctionCall(Radius=@n26.R, Density=@n26.R) {materialFunction=&RadialGradientExponential}
  n28 = ScalarParameter() {defaultValue=2, parameterName=TextureA_PowerToOpacity}
  n29 = Multiply(A=Multiply(A=ScalarParameter() {defaultValue=1.3, parameterName=TextureA_StrengthToOpacity}, B=Power(Base=@n15.R, Exp=@n28)), B=@n22)
  n30 = OneMinus(None=DotProduct(A=CameraVectorWS(), B=ComponentMask(None=@n6.Color) {b=true, g=true, r=true}))
  n31 = SceneTexelSize()
  n32 = Subtract(A=ParticlePositionWS(), B=CameraPositionWS())
  MaterialExpressionDivide_5 = Divide(
   A=Divide(A=ScalarParameter() {defaultValue=1.5, parameterName=Distance}, B=Multiply(A=@n31) {constB=1440})
   B=Clamp(None=MaterialFunctionCall(Vector 3=@n32) {materialFunction=/Engine/Functions/Engine_MaterialFunctions02/Utility/VectorLength}.V3 Length) {clampMode=CMODE_ClampMin, minDefault=30}
  )
  MaterialExpressionScalarParameter_1 = ScalarParameter() {defaultValue=6, parameterName="Distance Steps (S)"}
  MaterialExpressionScalarParameter_2 = ScalarParameter() {defaultValue=4, parameterName="Radial Steps (S)"}
  MaterialExpressionConstant_2 = Constant() {r=1}
  MaterialExpressionConstant_3 = Constant() {r=1}
  MaterialExpressionTextureObject_5 = TextureObject() {samplerType=SAMPLERTYPE_LinearGrayscale, texture=/Engine/EngineMaterials/Good64x64TilingNoiseHighFreq}
  MaterialExpressionConstant_1 = Constant() {r=1}
  MaterialExpressionConstant_0 = Constant() {r=0.62}
  MaterialExpressionScalarParameter_7 = ScalarParameter() {defaultValue=50, parameterName=DepthT}
  n33 = Custom!(
   Distance=@MaterialExpressionDivide_5
   DistanceSteps=@MaterialExpressionScalarParameter_1
   RadialSteps=@MaterialExpressionScalarParameter_2
   TempAARotation=@MaterialExpressionConstant_2
   DistanceMask=@MaterialExpressionConstant_3
   Tex=@MaterialExpressionTextureObject_5
   TempAADistance=@MaterialExpressionConstant_1
   RadialOffset=@MaterialExpressionConstant_0
   DivideF=@MaterialExpressionScalarParameter_7
   PDepth=@MaterialExpressionPixelDepth_0.R
  ) {
   code="float CurColor = 0;\r\nfloat2 BaseUV = MaterialFloat2(ScreenAlignedPosition(Parameters.ScreenPosition).xy);\r\nfloat2 NewUV = BaseUV;\r\nfloat StepSize = Distance / (int) DistanceSteps;\r\nfloat CurDistance = 0;\r\nfloat2 CurOffset = 0;\r\nfloat TwoPi = 6.283185;\r\nfloat Substep = 0;\r\nfloat2 ScenePixels=View.BufferSizeAndInvSize.xy*BaseUV;\r\nScenePixels+=View.TemporalAAParams.r;\r\nfloat2 RandomSamp = ((uint)(ScenePixels.x) + 2 * (uint)(ScenePixels.y)) % 5;\r\nRandomSamp+=Texture2DSample(Tex,TexSampler,ScenePixels);\r\nRandomSamp/=5;\r\nRandomSamp-=0.5;\r\nTempAARotation*=RandomSamp;\r\nTempAADistance*=StepSize*RandomSamp;\r\n\r\n\r\nint i=0;\r\nif (DistanceSteps<1)\r\n{\r\nreturn CalcSceneDepth(NewUV);\r\n}\r\nelse\r\n{\r\n//CurDistance += 0.5*StepSize;\r\nwhile ( i < (int) DistanceSteps)\r\n{\r\n\r\n//CurDistance+=StepSize;\r\nfor (int j = 0; j < (int) RadialSteps; j++)\r\n{\r\nCurOffset.x = cos(TwoPi*((TempAARotation+Substep) / RadialSteps));\r\nCurOffset.y = sin(TwoPi*((TempAARotation+Substep) / RadialSteps));\r\nCurOffset *=DistanceMask;\r\nNewUV.x = BaseUV.x + (CurOffset.x * (CurDistance+(RandomSamp*TempAADistance)));\r\nNewUV.y = BaseUV.y + (CurOffset.y * (CurDistance+(RandomSamp*TempAADistance)));\r\nfloat tmpdepth = CalcSceneDepth(NewUV)-PDepth;\r\nCurColor += (min(1-saturate(max(0,(tmpdepth))/DivideF)-0.99,0)+0.99)*saturate(sign(tmpdepth));\r\n//CurDistance+=(StepSize+(TempAADistance))/RadialSteps;\r\nSubstep++;\r\n}\r\nCurDistance+=StepSize;\r\nSubstep+=RadialOffset;\r\ni++;\r\n}\r\nfloat tmpdepth2 = CalcSceneDepth(BaseUV)-PDepth;\r\nfloat mask = (min(1-saturate(max(0,(tmpdepth2))/DivideF)-0.99,0)+0.99)*saturate(sign(tmpdepth2));\r\nMask = mask;\r\nCurColor = CurColor / ((int)DistanceSteps*(int)RadialSteps);\r\nreturn CurColor;\r\n}"
   desc="\n*** Original source before expression upgrade ***\nfloat3 CurColor;\r\nfloat2 BaseUV = MaterialFloat2(ScreenAlignedPosition(Parameters.ScreenPosition).xy);\r\nfloat2 NewUV = BaseUV;\r\nfloat StepSize = Distance / (int) DistanceSteps;\r\nfloat CurDistance=0;\r\nfloat2 CurOffset;\r\nfloat TwoPi = 6.283185;\r\nfloat Substep;\r\nfloat2 ScenePixels=View.RenderTargetSize*BaseUV;\r\nScenePixels+=View.TemporalAAParams.r;\r\nfloat2 RandomSamp = ((uint)(ScenePixels.x) + 2 * (uint)(ScenePixels.y)) % 5;\r\nRandomSamp+=Texture2DSample(Tex,TexSampler,ScenePixels);\r\nRandomSamp/=5;\r\nRandomSamp-=0.5;\r\nTempAARotation*=RandomSamp;\r\nTempAADistance*=StepSize*RandomSamp;\r\n\r\n\r\nint i=0;\r\nif (DistanceSteps<1)\r\n{\r\nreturn DecodeSceneColorForMaterialNode(NewUV);\r\n}\r\nelse\r\n{\r\n//CurDistance += 0.5*StepSize;\r\nwhile ( i < (int) DistanceSteps)\r\n{\r\n\r\n//CurDistance+=StepSize;\r\nfor (int j = 0; j < (int) RadialSteps; j++)\r\n{\r\nCurOffset.x = cos(TwoPi*((TempAARotation+Substep) / RadialSteps));\r\nCurOffset.y = sin(TwoPi*((TempAARotation+Substep) / RadialSteps));\r\nCurOffset *=DistanceMask;\r\nNewUV.x = BaseUV.x + (CurOffset.x * (CurDistance+(RandomSamp*TempAADistance)));\r\nNewUV.y = BaseUV.y + (CurOffset.y * (CurDistance+(RandomSamp*TempAADistance)));\r\nCurColor += DecodeSceneColorForMaterialNode(NewUV);\r\n//CurDistance+=(StepSize+(TempAADistance))/RadialSteps;\r\nSubstep++;\r\n}\r\nCurDistance+=StepSize;\r\nSubstep+=RadialOffset;\r\ni++;\r\n}\r\nCurColor = CurColor / ((int)DistanceSteps*(int)RadialSteps);\r\nreturn CurColor;\r\n}\n*** Original source before expression upgrade ***\nfloat3 CurColor;\r\nfloat2 BaseUV = MaterialFloat2(ScreenAlignedPosition(Parameters.ScreenPosition).xy);\r\nfloat2 NewUV = BaseUV;\r\nfloat StepSize = Distance / (int) DistanceSteps;\r\nfloat CurDistance=0;\r\nfloat2 CurOffset;\r\nfloat TwoPi = 6.283185;\r\nfloat Substep;\r\nfloat2 ScenePixels=Frame.RenderTargetSize*BaseUV;\r\nScenePixels+=Frame.TemporalAAParams.r;\r\nfloat2 RandomSamp = ((uint)(ScenePixels.x) + 2 * (uint)(ScenePixels.y)) % 5;\r\nRandomSamp+=Texture2DSample(Tex,TexSampler,ScenePixels);\r\nRandomSamp/=5;\r\nRandomSamp-=0.5;\r\nTempAARotation*=RandomSamp;\r\nTempAADistance*=StepSize*RandomSamp;\r\n\r\n\r\nint i=0;\r\nif (DistanceSteps<1)\r\n{\r\nreturn DecodeSceneColorForMaterialNode(NewUV);\r\n}\r\nelse\r\n{\r\n//CurDistance += 0.5*StepSize;\r\nwhile ( i < (int) DistanceSteps)\r\n{\r\n\r\n//CurDistance+=StepSize;\r\nfor (int j = 0; j < (int) RadialSteps; j++)\r\n{\r\nCurOffset.x = cos(TwoPi*((TempAARotation+Substep) / RadialSteps));\r\nCurOffset.y = sin(TwoPi*((TempAARotation+Substep) / RadialSteps));\r\nCurOffset *=DistanceMask;\r\nNewUV.x = BaseUV.x + (CurOffset.x * (CurDistance+(RandomSamp*TempAADistance)));\r\nNewUV.y = BaseUV.y + (CurOffset.y * (CurDistance+(RandomSamp*TempAADistance)));\r\nCurColor += DecodeSceneColorForMaterialNode(NewUV);\r\n//CurDistance+=(StepSize+(TempAADistance))/RadialSteps;\r\nSubstep++;\r\n}\r\nCurDistance+=StepSize;\r\nSubstep+=RadialOffset;\r\ni++;\r\n}\r\nCurColor = CurColor / ((int)DistanceSteps*(int)RadialSteps);\r\nreturn CurColor;\r\n}"
   outputType=CMOT_Float1
   additionalOutputs [{outputName=Mask, outputType=CMOT_Float1}]
   inputs [
    {inputName=Distance, input {expression=&self:MaterialExpressionDivide_5}}
    {
     inputName=DistanceSteps
     input {expression=&self:MaterialExpressionScalarParameter_1}
    }
    {
     inputName=RadialSteps
     input {expression=&self:MaterialExpressionScalarParameter_2}
    }
    {inputName=TempAARotation, input {expression=&self:MaterialExpressionConstant_2}}
    {inputName=DistanceMask, input {expression=&self:MaterialExpressionConstant_3}}
    {inputName=Tex, input {expression=&self:MaterialExpressionTextureObject_5}}
    {inputName=TempAADistance, input {expression=&self:MaterialExpressionConstant_1}}
    {inputName=RadialOffset, input {expression=&self:MaterialExpressionConstant_0}}
    {inputName=DivideF, input {expression=&self:MaterialExpressionScalarParameter_7}}
    {
     inputName=PDepth
     input {expression=&self:MaterialExpressionPixelDepth_0, mask=1, maskR=1}
    }
   ]
  }
  n34 = Multiply(A=Power(Base=Saturate(None=@n30), Exp=ScalarParameter() {defaultValue=4, parameterName=150}), B=@n33.Mask)
  n35 = ScalarParameter() {parameterName=St}
  n36 = ScalarParameter() {defaultValue=0.6, parameterName=Test}
  n37 = Saturate(None=Subtract(A=@n33.return, B=Multiply(A=@n33.Mask, B=@n36)))
  n38 = Multiply(A=@n37, B=ScalarParameter() {defaultValue=3, parameterName=Strengt2})
  n39 = Power(Base=OneMinus(None=@n37), Exp=ScalarParameter() {defaultValue=8, parameterName=T4})
  n40 = ScalarParameter() {defaultValue=2.5, parameterName=T2}
  n41 = Saturate(None=Multiply(A=Power(Base=@n33.Mask, Exp=@n40), B=ScalarParameter() {defaultValue=10, parameterName=T3}) {constB=1000})
  n42 = Multiply(
   A=Multiply(A=ScalarParameter() {defaultValue=1.5, parameterName=Strengt}, B=LinearInterpolate(A=@n38, B=@n39, Alpha=@n41) {constB=0})
   B=Saturate(None=Max(A=@n18.R))
  )
  n43 = Add(A=Add(A=Multiply(A=@n2.T_Opacity, B=@n29), B=Multiply(A=@n34, B=@n35)), B=Multiply(A=Saturate(None=@n42), B=@n2.Glow_Opacity))
  n44 = Multiply(
   A=@n1.A
   B=DepthFade(Opacity=Multiply(A=@n27.RadialGradientExponential, B=@n43), FadeDistance=ScalarParameter() {defaultValue=5, parameterName=DF})
  )
  ShadingPathSwitch(Default=@n44, Deferred=@n44, Forward=@n25, Mobile=@n25)
 ]
}
