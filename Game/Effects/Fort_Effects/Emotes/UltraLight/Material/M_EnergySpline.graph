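# Unlit, two-sided translucent material for energy spline effects on Niagara mesh particles in the UltraLight emote.
# Emissive multiplies noise-distorted TextureB by tint, hue shift along V, a vertical brightness gradient and particle color.
# Opacity combines TextureB, plasma TextureA, edge masks and center falloff, with depth fade and optional view-angle fade.
# SplineThicken drives tapered width via world position offset, while a dynamic parameter offsets texture panning time.
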Material /Game/Effects/Fort_Effects/Emotes/UltraLight/Material/M_EnergySpline
graph {
 properties {
  bUsedWithNiagaraMeshParticles=true
  blendMode=BLEND_Translucent
  pixelDepthOffsetMode=PDOM_Legacy
  shadingModel=MSM_Unlit
  twoSided=true
  usedShadingModels=Unlit
 }
 nodes [
  n1 = ParticleColor()
  n2 = MaterialFunctionCall(Float2=TextureCoordinate()) {materialFunction=&BreakOutFloat2Components}
  n3 = VectorParameter() {parameterName=BottomScale, defaultValue=<rgba 0 1 6 20>}
  n4 = Saturate(None=Multiply(A=Add(A=@n2.G, B=@n3.R), B=@n3.G))
  n5 = Add(A=Multiply(A=Power(Base=@n4, Exp=@n3.B), B=@n3.A))
  n6 = TextureCoordinate()
  n7 = VectorParameter() {
   group=---TextureB
   parameterName=TextureB_UVControl_Mark:********************
   channelNames {a=PanV, b=PanU, g=ScaleV, r=ScaleU}
   defaultValue=<rgba 1 1 0 0>
  }
  n8 = Add(A=Time() {bOverride_Period=true, period=13333}, B=DynamicParameter() {paramNames [TimeInput,Param2,Param3,Param4]}.TimeInput)
  n9 = VectorParameter() {
   group=---NoiseA
   parameterName=NoiseA_UVControl_Mark:********************
   channelNames {a=PanV, b=PanU, g=ScaleV, r=ScaleU}
   defaultValue=<rgba 1 1 0 0>
  }
  n10 = MaterialFunctionCall(
   Texture=TextureObjectParameter() {group=---NoiseA, parameterName=NoiseA_Texture, sortPriority=31}
   UVControl_RGB=@n9
   UVControl_A=@n9
   "[Override]TimeInput"=@n8
  ) {materialFunction=&FXCommon_TextureSample}
  n11 = AppendVector(A=@n10.R, B=@n10.R)
  n12 = VectorParameter() {
   group=---NoiseA
   parameterName=NoiseA_ValueControl_Mark:--------------------
   sortPriority=33
   channelNames {a=StrengthV, b=StrengthU, g=StrengthOffsetV, r=StrengthOffsetU}
   defaultValue=<rgba -0.5 -0.5 0 0>
  }
  n13 = MaterialFunctionCall(
   Texture=TextureObjectParameter() {group=---TextureB, parameterName=TextureB_Texture, sortPriority=31}
   UVControl_RGB=@n7
   UVControl_A=@n7
   "[Override]TimeInput"=@n8
   "[Override]UVInput"=Add(
    A=TextureCoordinate()
    B=MaterialFunctionCall!(NoiseValue=@n11, Control_RGB=@n12, Control_A=@n12) {materialFunction=/Game/Effects/Niagara/MasterMaterials/Functions/Common/FXCommon_NoiseControl}.NoiseValue
   )
  ) {materialFunction=&FXCommon_TextureSample}
  n14 = VectorParameter() {parameterName=TextureB_Refine_Brightness, defaultValue=<rgba 10 2000 0.3 0.5>}
  n15 = MaterialFunctionCall(Texture=@n13.R, RGB=@n14.RGB, A=@n14.RGB) {materialFunction=&MF_FX_Refine}
  n16 = MaterialFunctionCall(
   Hue Shift Percentage=Multiply(
    A=MaterialFunctionCall(Float2=@n6) {materialFunction=&BreakOutFloat2Components}.G
    B=ScalarParameter() {parameterName=HueShift}
   )
   Texture=Multiply(A=Add(A=@n15.Result), B=VectorParameter() {parameterName=C, defaultValue=<rgba 0 0 0 1>}.RGB)
  ) {materialFunction=&HueShift}
  MaterialFunctionCall(Emissive=Multiply(A=@n1.RGB, B=Multiply(A=@n5, B=@n16.Result))) {materialFunction=&MF_FX_EmissiveAdjustments}
  n17 = MaterialFunctionCall(Float2=TextureCoordinate()) {materialFunction=&BreakOutFloat2Components}
  n18 = Multiply(A=@n17.R) {constB=0.5}
  n19 = Multiply(A=OneMinus(None=Sine(None=@n18)), B=ScalarParameter() {defaultValue=2, parameterName=CenterFalloff})
  n20 = ScalarParameter() {parameterName=CenterBase}
  n21 = LinearInterpolate(
   A=Saturate(None=Add(A=@n19, B=@n20))
   Alpha=Power(Base=OneMinus(None=@n17.G), Exp=ScalarParameter() {defaultValue=2, parameterName=CenterFalloffAreaPower})
  ) {constA=1}
  n22 = MaterialFunctionCall(Float2=TextureCoordinate()) {materialFunction=&BreakOutFloat2Components}
  n23 = VectorParameter() {group=---EdgeMask, parameterName=EdgeMask_U_Invert, defaultValue=<rgba 100 1 0 1>}
  n24 = VectorParameter() {group=---EdgeMask, parameterName=EdgeMask_U_NoInvert, defaultValue=<rgba 100 1 0 1>}
  n25 = VectorParameter() {group=---EdgeMask, parameterName=EdgeMask_V, defaultValue=<rgba 100 1 0 1>}
  n26 = ScalarParameter() {parameterName=TextureA_Strength}
  n27 = TextureObjectParameter() {
   group=---TextureA
   parameterName=TextureA_Texture
   samplerType=SAMPLERTYPE_LinearColor
   texture=/Game/Effects/Fort_Effects/Emotes/UltraLight/Texture/T_Plasma_A_C
  }
  n28 = VectorParameter() {group=---TextureA, parameterName=TextureA_UVOffset, channelNames {a="", b="", g=OffsetV, r=OffsetU}, defaultValue=<rgba 1 1 0 0>}
  n29 = VectorParameter() {
   group=---TextureA
   parameterName=TextureA_UVControl_Mark:********************
   channelNames {a=PanV, b=PanU, g=ScaleV, r=ScaleU}
   defaultValue=<rgba 1 1 0 0>
  }
  n30 = MaterialFunctionCall(
   Texture=@n27
   AdditionalUVOffset=AppendVector(A=@n28.OffsetU, B=@n28.OffsetU)
   UVControl_RGB=@n29
   UVControl_A=@n29
   "[Override]TimeInput"=@n8
  ) {materialFunction=&FXCommon_TextureSample}
  n31 = Add(
   A=Power(Base=Add(A=@n30.R) {constB=0.01}, Exp=ScalarParameter() {defaultValue=1, parameterName=TextureA_Power})
   B=ScalarParameter() {parameterName=TextureA_BaseStr}
  )
  n32 = Multiply(
   A=Multiply(
    A=MaterialFunctionCall!(UorV=@n22.G, Invert_EdgeControl=@n23.RGB, NoInvert_EdgeControl=@n24.RGB) {materialFunction=/Game/Effects/Niagara/MasterMaterials/Functions/Common/FXCommon_EdgeMask_Ex}.Mask
    B=MaterialFunctionCall!(UorV=@n22.R, EdgeControl=@n25.RGB) {materialFunction=/Game/Effects/Niagara/MasterMaterials/Functions/Common/FXCommon_EdgeMask_Simple}.Mask
   )
   B=Add(A=@n13.R, B=Multiply(A=@n26, B=@n31))
  )
  n33 = DepthFade(
   Opacity=Multiply(A=ScalarParameter() {defaultValue=1, parameterName=TotalOpacityScale}, B=Multiply(A=@n21, B=@n32))
   FadeDistance=ScalarParameter() {defaultValue=10, parameterName=DF}
  )
  n34 = MaterialFunctionCall("FadeSensitivity (optional)"=ScalarParameter() {parameterName="FadeSensitivity (optional) (S)"}) {materialFunction=&FadeBasedOnViewAngle}
  Multiply(A=LinearInterpolate(A=@n33, B=Multiply(A=@n33, B=@n34.Result), Alpha=ScalarParameter() {parameterName=UseViewAngleFade}), B=@n1.A)
  n35 = TextureCoordinate()
  n36 = LinearInterpolate(
   A=ScalarParameter() {group=---Spline, parameterName=Spline_WidthTip}
   B=ScalarParameter() {group=---Spline, parameterName=Spline_WidthBase}
   Alpha=Power(
    Base=MaterialFunctionCall(Float2=@n35) {materialFunction=&BreakOutFloat2Components}.G
    Exp=ScalarParameter() {group=---Spline, parameterName=Spline_GradientPower}
   )
  )
  Multiply(
   A=MaterialFunctionCall(
    WidthBase=@n36
    WidthTip=@n36
    Expand U or V UV Channel=StaticBoolParameter() {defaultValue=true, parameterName="Expand U or V UV Channel (SB)"}
   ) {materialFunction=&SplineThicken}.WorldPositionOffset
   B=ScalarParameter() {defaultValue=1, parameterName=WPO_Strength}
  )
 ]
}
