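# Additive unlit Niagara sprite material for a flare in the UltraLight emote effects, with an animated horizontal sway.
# A time-driven sine offsets the texture's U coordinate, with a sharp, adjustable split along V making the two halves swing in opposite directions.
# Dynamic parameters control swing speed, strength, split position and sharpness.
# The sampled flare is partially desaturated, tinted by particle color and fed through the FX emissive adjustment function.
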Material /Game/Effects/Fort_Effects/Emotes/UltraLight/Material/M_Flare
graph {
 properties {bUsedWithNiagaraSprites=true, blendMode=BLEND_Additive, pixelDepthOffsetMode=PDOM_Legacy, shadingModel=MSM_Unlit, usedShadingModels=Unlit}
 nodes [
  n1 = MaterialFunctionCall(Float2=TextureCoordinate()) {materialFunction=&BreakOutFloat2Components}
  n2 = DynamicParameter() {defaultValue=<rgba 0.2 0.025 -0.96 50>, paramNames [SwingSpeed, SwingStrength, SplitPos, SplitSharpness]}
  n3 = Add(A=Multiply(A=@n1.G) {constB=2}, B=@n2.SplitPos) {constB=-0.96}
  Ceil(None=@n3)
  n4 = ParticleColor()
  DepthFade(Opacity=@n4.A, FadeDistance=Constant() {r=10})
  n5 = Multiply(A=@n3, B=@n2.SplitSharpness) {constB=50}
  n6 = Multiply(A=Time(), B=@n2.SwingSpeed)
  n7 = Multiply(A=LinearInterpolate(Alpha=Saturate(None=@n5)) {constA=-1}, B=Multiply(A=Sine(None=@n6), B=@n2.SwingStrength) {constB=0.025}) {constB=0.01}
  n8 = TextureSampleParameter2D(UVs=AppendVector(A=Add(A=@n1.R, B=@n7), B=@n1.G)) {parameterName=Tex, texture=/Game/Effects/Fort_Effects/Emotes/UltraLight/Texture/T_Flare3}
  n9 = ScalarParameter() {defaultValue=0.5, parameterName=desat}
  MaterialFunctionCall(Emissive=Multiply(A=@n4.RGB, B=Desaturation(None=@n8.RGB, Fraction=@n9))) {materialFunction=&MF_FX_EmissiveAdjustments}
  Multiply(A=Desaturation(None=@n8.RGB, Fraction=Constant() {r=1}), B=@n4.A)
  Normalize(VectorInput=@n8.RGB)
 ]
}
