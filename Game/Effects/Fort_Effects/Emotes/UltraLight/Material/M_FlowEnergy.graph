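# M_FlowEnergy is a translucent, unlit, two-sided Niagara mesh-particle material for flowing energy in the UltraLight emote effects.
# Two noise samples distort TextureA's UVs, sharing time from a toggleable clock plus Dynamic Parameter offset.
# The refined result drives emissive, colored by a curve atlas or solid color and particle color.
# Opacity combines refined texture, Fresnel, edge mask, depth fade and particle alpha; optional WPO displaces along normals.
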Material /Game/Effects/Fort_Effects/Emotes/UltraLight/Material/M_FlowEnergy
graph {
 properties {
  bUsedWithNiagaraMeshParticles=true
  blendMode=BLEND_Translucent
  pixelDepthOffsetMode=PDOM_Legacy
  shadingModel=MSM_Unlit
  twoSided=true
  usedShadingModels=Unlit
 }
 nodes [
  n1 = TextureObjectParameter() {
   group=---TextureA
   parameterName=TextureA_Texture
   samplerType=SAMPLERTYPE_LinearColor
   texture=/Game/Effects/Fort_Effects/Emotes/UltraLight/Texture/T_Plasma_A_C
  }
  n2 = VectorParameter() {
   group=---TextureA
   parameterName=TextureA_UVControl_Mark:********************
   channelNames {a=PanV, b=PanU, g=ScaleV, r=ScaleU}
   defaultValue=<rgba 1 1 0 0>
  }
  n3 = Add(
   A=Multiply(A=ScalarParameter() {defaultValue=1, parameterName=DisableMatTime}, B=Time() {bOverride_Period=true, period=13333})
   B=DynamicParameter() {paramNames [TimeInput,Param2,Param3,Param4]}.TimeInput
  )
  n4 = TextureObjectParameter() {group=---NoiseA, parameterName=NoiseA_Texture, texture=/Game/Characters/Player/Male/Medium/Bodies/M_MED_Viper/Textures/Glitch/T_SoftNoise}
  n5 = VectorParameter() {
   group=---NoiseA
   parameterName=NoiseA_UVControl_Mark:********************
   channelNames {a=PanV, b=PanU, g=ScaleV, r=ScaleU}
   defaultValue=<rgba 1 1 0 0>
  }
  n6 = VectorParameter() {parameterName=UVInvert, defaultValue=<rgba 1 1 0 1>}
  n7 = Multiply(A=TextureCoordinate(), B=AppendVector(A=@n6.R, B=@n6.R))
  n8 = VectorParameter() {
   group=---NoiseA
   parameterName=NoiseA_ValueControl_Mark:--------------------
   sortPriority=33
   channelNames {a=StrengthV, b=StrengthU, g=StrengthOffsetV, r=StrengthOffsetU}
   defaultValue=<rgba -0.5 -0.5 0 0>
  }
  n9 = MaterialFunctionCall!(
   NoiseValue=MaterialFunctionCall(Texture=@n4, UVControl_RGB=@n5, UVControl_A=@n5, "[Override]TimeInput"=@n3, "[Override]UVInput"=@n7) {materialFunction=&FXCommon_TextureSample}.RG
   Control_RGB=@n8
   Control_A=@n8
  ) {materialFunction=/Game/Effects/Niagara/MasterMaterials/Functions/Common/FXCommon_NoiseControl}
  n10 = VectorParameter() {parameterName=UVInvert, defaultValue=<rgba 1 1 0 1>}
  n11 = Multiply(A=TextureCoordinate(), B=AppendVector(A=@n10.R, B=@n10.R))
  n12 = TextureObjectParameter() {
   group=---NoiseB
   parameterName=NoiseB_Texture
   samplerType=SAMPLERTYPE_LinearGrayscale
   texture=/Game/Effects/Fort_Effects/Emotes/UltraLight/Texture/T_DirectionalNoise_Greyscale
  }
  n13 = VectorParameter() {
   group=---NoiseB
   parameterName=NoiseB_UVControl_Mark:********************
   channelNames {a=PanV, b=PanU, g=ScaleV, r=ScaleU}
   defaultValue=<rgba 1 1 0 0>
  }
  n14 = MaterialFunctionCall(Texture=@n12, UVControl_RGB=@n13, UVControl_A=@n13, "[Override]TimeInput"=@n3, "[Override]UVInput"=@n7) {materialFunction=&FXCommon_TextureSample}
  n15 = AppendVector(A=@n14.R, B=@n14.R)
  n16 = VectorParameter() {
   group=---NoiseB
   parameterName=NoiseB_ValueControl_Mark:--------------------
   sortPriority=33
   channelNames {a=StrengthV, b=StrengthU, g=StrengthOffsetV, r=StrengthOffsetU}
   defaultValue=<rgba -0.5 -0.5 0 0>
  }
  n17 = MaterialFunctionCall(
   Texture=@n1
   UVControl_RGB=@n2
   UVControl_A=@n2
   "[Override]TimeInput"=@n3
   "[Override]UVInput"=Add(
    A=Add(A=@n9.NoiseValue, B=@n11)
    B=MaterialFunctionCall(NoiseValue=@n15, Control_RGB=@n16, Control_A=@n16) {materialFunction=&FXCommon_NoiseControl}.NoiseValue
   )
  ) {materialFunction=&FXCommon_TextureSample}
  n18 = Power(
   Base=LinearInterpolate(
    A=@n17.RGB
    B=MaterialFunctionCall(X=@n17.R, Y=@n17.R, Z=@n17.R) {materialFunction=&MakeFloat3}.Result
    Alpha=ScalarParameter() {parameterName=IsGreyscaleTexture}
   )
   Exp=ScalarParameter() {defaultValue=1, parameterName=TexturePower}
  )
  n19 = Normalize(VectorInput=@n18)
  n20 = MaterialFunctionCall(Float3=@n18) {materialFunction=&BreakOutFloat3Components}
  n21 = CurveAtlasRowParameter(CurveTime=@n20.R) {
   atlas=/Game/Effects/Fort_Effects/Emotes/UltraLight/Texture/Curve/CLA_UltraLight
   curve=/Game/Effects/Fort_Effects/Emotes/UltraLight/Texture/Curve/CL_UltraLight_Energy
  }
  n22 = VectorParameter() {parameterName=Refine2, defaultValue=<rgba 10 2000 0.3 0.5>}
  n23 = MaterialFunctionCall(Texture=@n20.R, RGB=@n22.RGB, A=@n22.RGB) {materialFunction=&MF_FX_Refine}
  n24 = Multiply(
   A=StaticSwitchParameter(True=VectorParameter() {parameterName=Color, defaultValue=<rgba 1 0 0 1>}.RGB, False=@n21) {parameterName=SolidColor}
   B=Multiply(B=Add(A=@n23.Result)) {constA=0.5}
  )
  n25 = ParticleColor()
  MaterialFunctionCall(Emissive=Multiply(A=Multiply(A=@n19, B=@n24), B=@n25.RGB)) {materialFunction=&MF_FX_EmissiveAdjustments}
  n26 = MaterialFunctionCall!(
   UorV=MaterialFunctionCall(Float2=TextureCoordinate()) {materialFunction=&BreakOutFloat2Components}.G
   Invert_EdgeControl=VectorParameter() {parameterName=EdgeControl_Invert, defaultValue=<rgba 0 1 1 1>}.RGB
   NoInvert_EdgeControl=VectorParameter() {parameterName=EdgeControl_NoInvert, defaultValue=<rgba 0 1 1 1>}.RGB
  ) {materialFunction=/Game/Effects/Niagara/MasterMaterials/Functions/Common/FXCommon_EdgeMask_Ex}
  n27 = ScalarParameter() {parameterName=OpenEdgeFade}
  n28 = VectorParameter() {parameterName=Refine, defaultValue=<rgba 3 1 0.7 0.5>}
  n29 = Fresnel(ExponentIn=ScalarParameter() {defaultValue=1, group=Fresnel, parameterName=Fresnel_Out_ExA}, BaseReflectFractionIn=Constant())
  n30 = Multiply(
   A=MaterialFunctionCall(Texture=@n20.R, RGB=@n28.RGB, A=@n28.RGB) {materialFunction=&MF_FX_Refine}.Result
   B=Power(Base=OneMinus(None=@n29), Exp=ScalarParameter() {defaultValue=1, group=Fresnel, parameterName=Fresnel_Out_ExB})
  )
  n31 = ScalarParameter() {defaultValue=1, group=Fresnel, parameterName=Fresnel_Inner_ExA}
  n32 = Constant()
  n33 = Power(
   Base=OneMinus(None=Fresnel(ExponentIn=@n31, BaseReflectFractionIn=@n32))
   Exp=ScalarParameter() {defaultValue=1, group=Fresnel, parameterName=Fresnel_Inner_ExB}
  )
  n34 = LinearInterpolate(A=@n30, B=Multiply(A=OneMinus(None=@n33), B=@n30), Alpha=ScalarParameter() {group=Fresnel, parameterName=OpenInnerFade})
  n35 = ScalarParameter() {parameterName=DapthFade}
  Multiply(A=@n25.A, B=Multiply(A=LinearInterpolate(B=@n26.Mask, Alpha=@n27) {constA=1}, B=DepthFade(Opacity=@n34, FadeDistance=@n35)))
  ScalarParameter() {parameterName=WPOStrength}
  n36 = TextureObjectParameter() {
   group=---WPOA
   parameterName=WPOA_Texture
   samplerType=SAMPLERTYPE_LinearColor
   texture=&T_Plasma_A_C
  }
  n37 = VectorParameter() {
   group=---TextureA
   parameterName=WPOA_UVControl_Mark:********************
   channelNames {a=PanV, b=PanU, g=ScaleV, r=ScaleU}
   defaultValue=<rgba 1 1 0 0>
  }
  n38 = MaterialFunctionCall(Texture=@n36, UVControl_RGB=@n37, UVControl_A=@n37, "[Override]TimeInput"=@n3, "[Override]UVInput"=@n11) {materialFunction=&FXCommon_TextureSample}
  n39 = VertexNormalWS()
  StaticSwitchParameter(True=Multiply(A=ScalarParameter() {parameterName=WPOA_Strength}, B=Multiply(A=@n38.R, B=@n39)), False=Constant()) {parameterName=UseWPO}
 ]
}
