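# Additive Niagara sprite material for a simple flare in the UltraLight emote effects.
# It multiplies particle color by the red channel of a camera flare texture and passes the result through the MF_FX_EmissiveAdjustments function.
# A DepthFade node takes particle alpha as opacity and the first dynamic parameter channel as fade distance, allowing per-particle soft intersections.
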Material /Game/Effects/Fort_Effects/Emotes/UltraLight/Material/M_SimpleFlare
graph {
 properties {bUsedWithNiagaraSprites=true, blendMode=BLEND_Additive, pixelDepthOffsetMode=PDOM_Legacy, usedShadingModels=Default Lit}
 nodes [
  n1 = ParticleColor()
  DepthFade(Opacity=@n1.A, FadeDistance=DynamicParameter().R)
  MaterialFunctionCall(Emissive=Multiply(A=@n1.RGB, B=TextureSample() {texture=/Game/Environments/FrontEnd/Textures/T_EFX_CameraFlare_02}.R)) {materialFunction=&MF_FX_EmissiveAdjustments}
 ]
}
